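# Custom character material for the Innovator Female glider, adding wind-blown cloth motion through BlackMonday-group parameters.
# Panning world-aligned fold normals, projected from object-relative or rotated pre-skinned position, blend into the CharacterSkinFX normal.
# Time-scrolled curl noise, plus an optional high-frequency sine wave via static switch, drives World Position Offset.
# An FX canvas texture masks both effects before PostFX; a second swirl-noise normal layer is left unconnected.
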Material /BRCosmetics/Gadgets/Assets/VinderTech_GliderChute/Glider_Innovator_Female/Materials/M_Glider_Innovator_Female_Custom
graph {
 properties {
  bUseMaterialAttributes=true
  bUsedWithMeshDeformer=true
  bUsedWithMorphTargets=true
  bUsedWithSkeletalMesh=true
  pixelDepthOffsetMode=PDOM_Legacy
  refractionMethod=RM_IndexOfRefraction
  shadingModel=MSM_FromMaterialExpression
  subsurfaceProfile=&SS_HeroSkin_02
  usedShadingModels=Default Lit | Subsurface | Clear Coat | Subsurface Profile | Cloth | Eye
 }
 nodes [
  ComponentMask(None=NamedRerouteUsage()) {b=true}
  n1 = MaterialFunctionCall() {materialFunction=&BaseFN_CharacterFunction_PreFX}
  n2 = GetMaterialAttributes(None=MaterialFunctionCall(
    PreskinnedPosition=@n1.Pre skinned Local Position
    Gmap=@n1.Pre skinned Local Position
    MaterialAttributes=@n1.Pre skinned Local Position
    AO=@n1.Pre skinned Local Position
   ) {materialFunction=&CharacterSkinFX}.Result) {attributeGetTypes [#1, #2]}
  n3 = TextureSample() {samplerType=SAMPLERTYPE_Normal, texture=&FlatNormalBlue}
  n4 = Subtract(A=ObjectPositionWS(), B=WorldPosition().XYZ)
  n5 = MaterialFunctionCall!(
   VectorToRotate=NamedRerouteUsage()
   LookAtVector=VectorParameter() {group=BlackMonday, parameterName=DirectionVector, defaultValue=<rgba 1 0 0 1>}.RGB
   RestingVector
  ) {materialFunction=/Engine/Functions/Engine_MaterialFunctions01/Vectors/RotateVector}
  n6 = LinearInterpolate(
   A=Transform(None=@n4) {transformSourceType=TRANSFORMSOURCE_World, transformType=TRANSFORM_Local}
   B=VertexInterpolator(VS=@n5.Result).PS
   Alpha=ScalarParameter() {group=BlackMonday, parameterName=UsePreskinnedUVs}
  )
  n7 = VectorParameter() {group=BlackMonday, parameterName=FoldsPanSpeed, defaultValue=<rgba -0.2 -1.5 0 1>}
  n8 = ComponentMask(None=MaterialFunctionCall!(
    TextureObject=TextureObject() {samplerType=SAMPLERTYPE_Normal, texture=/Game/Gadgets/Assets/VinderTech_GliderChute/Glider_BlackMondayCape/Textures/T_Folds_N}
    TextureSize=VectorParameter() {group=BlackMonday, parameterName=FoldNormalTiling, defaultValue=<rgba 256 128 128 1>}.RGB
    Normal
    WorldSpace
    ProjectionTransitionContrast
    WorldPosition=@n6
    Use High Quality Normals
    Pan_YZ
    Pan_XZ
    Pan_XY=ComponentMask(None=@n7.RGB) {g=true, r=true}
   ) {materialFunction=/Game/Weapons/WeaponSkins/Wraps/MaterialFunctions/WorldAlignedNormal_ComplexPan}.XYZ Texture) {b=true, g=true, r=true}
  n9 = ScalarParameter() {defaultValue=1, group=BlackMonday, parameterName=RippleNormalIntensity, sliderMax=1}
  n10 = TextureSample() {
   samplerType=SAMPLERTYPE_Masks
   texture=/BRCosmetics/Gadgets/Assets/VinderTech_GliderChute/Glider_Innovator_Female/Textures/T_Glider_Innovator_Female_FX_Canvas
  }
  n11 = Normalize(VectorInput=LinearInterpolate(A=LinearInterpolate(A=@n3.RGB, B=@n8, Alpha=@n9), B=@n3.RGB, Alpha=@n10.G) {constB=0})
  n12 = Normalize(VectorInput=@n4)
  n13 = MaterialFunctionCall(Float3=Divide(B=VectorParameter() {group=BlackMonday, parameterName=DisplacementNoiseScale, defaultValue=<rgba 512 256 256 1>}.RGB)) {materialFunction=&BreakOutFloat3Components}
  n14 = Add(A=@n13.G, B=Multiply(A=Time(), B=ScalarParameter() {defaultValue=2, group=BlackMonday, parameterName=WindFlapSpeed}))
  n15 = Multiply(
   A=VectorNoise(World Position=MaterialFunctionCall(X=@n13.R, Y=@n14, Z=@n13.R) {materialFunction=&MakeFloat3}.Result) {noiseFunction=VNF_CurlALU}
   B=Multiply(B=VectorParameter() {group=BlackMonday, parameterName=VertexDIsplacement_Amt, defaultValue=<rgba 0.25 0.25 0.5 1>}.RGB) {constA=1}
  )
  n16 = ChannelMaskParameter(None=@n5.Result) {group=BlackMonday, maskChannel=Green, parameterName=HighFrequencyWPO_Direction, defaultValue=<rgba 0 1 0 0>}
  n17 = ScalarParameter() {defaultValue=64, group=BlackMonday, parameterName=HighFrequency}
  n18 = Time()
  n19 = ScalarParameter() {defaultValue=2, group=BlackMonday, parameterName=HighFrequencyWPO_Speed}
  n20 = Sine(None=Add(A=Divide(A=@n16.R, B=@n17), B=Multiply(A=@n18, B=@n19)))
  n21 = VectorParameter() {group=BlackMonday, parameterName=HighFrequencyWPO_Offset, defaultValue=<rgba 3 2 4 1>}
  n22 = StaticSwitchParameter(True=Add(A=@n15, B=Multiply(A=@n20, B=@n21.RGB)), False=@n15) {group=BlackMonday, parameterName=AddHighFrequencyWPO}
  n23 = LinearInterpolate(B=LinearInterpolate(A=Add(A=@n12, B=@n22), Alpha=@n10.G), Alpha=@n10.R)
  n24 = MaterialFunctionCall(
   AO=@n1.AO
   Distance Blend=@n1.AO
   Fresnel=@n1.AO
   Gmap=@n1.AO
   Pre Skinned Local Position=@n1.AO
   Material=SetMaterialAttributes!(
    MaterialAttributes=@n2.MaterialAttributes
    Normal=MaterialFunctionCall(BaseNormal=@n2.Normal, AdditionalNormal=@n11) {materialFunction=&BlendAngleCorrectedNormals}.Result
    World Position Offset=Add(A=@n2.World Position Offset, B=@n23)
    Base Color
   ) {attributeSetTypes [#1, #2, #3]}
  ) {materialFunction=&BaseFN_CharacterFunction_PostFX}
  MaterialFunctionCall(In=@n24.Result) {materialFunction=&CINE_MF_CharPostFX}
  n25 = TextureObject() {samplerType=SAMPLERTYPE_Normal, texture=/Game/Global/Textures/Default/Noise/T_SwirlNoise_N}
  n26 = VectorParameter() {group=BlackMonday, parameterName=FoldPerlinTiling, defaultValue=<rgba 256 128 128 1>}
  n27 = ComponentMask(None=VectorParameter() {group=BlackMonday, parameterName=FoldsPanSpeed2, defaultValue=<rgba 0.2 -2 0 1>}.RGB) {g=true, r=true}
  MaterialFunctionCall(
   BaseNormal=@n8
   AdditionalNormal=ComponentMask(None=MaterialFunctionCall(TextureObject=@n25, TextureSize=@n26.RGB, WorldPosition=@n6, Pan_XY=@n27) {materialFunction=&WorldAlignedNormal_ComplexPan}.XYZ Texture) {b=true, g=true, r=true}
  ) {materialFunction=&BlendAngleCorrectedNormals}
  Multiply(A=@n21.RGB, B=VertexColor().G)
  NamedRerouteDeclaration(None=@n1.Pre skinned Local Position) {name=PreskinnedLocalPosition, nodeColor=<rgba 0 0.0391 1 1>}
  OneMinus(None=ComponentMask() {g=true})
 ]
}
